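# Master material for the York (GhostBusters) backpack, with emissive effects driven by a Kills parameter.
# An FX texture's red channel offsets a time-based sine pulse and feeds sphere masks that widen with Kills to reveal emissives.
# A UV-based mask switches to a gas bar colored from a curve atlas, with the green channel masking its glow and tint.
# The results override emissive and base color between the shared character PreFX and PostFX functions.
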Material /BRCosmetics/Accessories/FORT_Backpacks/M_MED_York_Backpack/Materials/MAT_M_MED_York_Backpack_Master
graph {
 properties {
  bUseMaterialAttributes=true
  bUsedWithMeshDeformer=true
  bUsedWithSkeletalMesh=true
  pixelDepthOffsetMode=PDOM_Legacy
  usedShadingModels=Default Lit
 }
 nodes [
  n1 = MaterialFunctionCall() {materialFunction=&BaseFN_CharacterFunction_PreFX}
  n2 = GetMaterialAttributes(None=@n1.Result) {attributeGetTypes [#1, #2]}
  n3 = Time()
  n4 = TextureSampleParameter2D() {
   group=GhostBusters
   parameterName=FX Texture
   samplerType=SAMPLERTYPE_Masks
   texture=/BRCosmetics/Accessories/FORT_Backpacks/M_MED_York_Backpack/Textures/T_Backpack_Male_York_FX
  }
  n5 = Multiply(A=@n4.R) {constB=-0.5}
  n6 = Sine(None=Multiply(A=Add(A=@n3, B=@n5)) {constB=0.5})
  n7 = Constant()
  n8 = ScalarParameter() {group=GhostBusters, parameterName=Kills}
  n9 = Multiply(A=@n2.Emissive Color, B=SphereMask(A=@n7, B=@n4.R, Radius=Multiply(A=@n8) {constB=0.26}) {attenuationRadius=0.01})
  n10 = Multiply(A=@n9, B=@n6)
  n11 = Add(
   A=Multiply(A=@n2.Emissive Color, B=Multiply(A=@n6) {constB=0.1, desc=strength of pulsing emissives}) {constB=0.1}
   B=Add(A=@n9, B=Multiply(A=@n10) {constB=0.1})
  )
  n12 = Add(A=Add(A=Multiply(A=@n6) {constB=0.05}, B=Multiply(A=@n8) {constB=0.2})) {constB=0.05}
  n13 = CurveAtlasRowParameter(CurveTime=OneMinus(None=@n4.R)) {
   atlas=&AthenaCurveAtlas01
   curve=/Game/Characters/Player/Male/Medium/Bodies/M_MED_BattleSuit/FX/BattleSuit_Blue
   defaultValue=82
  }
  n14 = Multiply(A=Multiply(A=SphereMask(A=@n7, B=@n4.R, Radius=@n12) {attenuationRadius=0.01, hardnessPercent=0}, B=@n13), B=@n4.G) {constB=0.3}
  n15 = SphereMask(A=ComponentMask(None=TextureCoordinate()) {g=true}, B=Constant()) {attenuationRadius=0.3}
  n16 = LinearInterpolate(
   A=LinearInterpolate(A=Multiply(A=@n2.Base Color) {constB=0.2}, B=Constant3Vector() {constant=<rgba 0.4 0.002 0.0005 0>}) {constAlpha=0.1}
   B=Multiply(A=@n2.Base Color) {constB=2, desc=gasBar}
   Alpha=@n15
  )
  MaterialFunctionCall(
   AO=@n1.AO
   Distance Blend=@n1.AO
   Fresnel=@n1.AO
   Material=SetMaterialAttributes!(
    MaterialAttributes=@n2.MaterialAttributes
    Emissive Color=LinearInterpolate(A=@n11, B=@n14, Alpha=@n15)
    Base Color=LinearInterpolate(A=@n2.Base Color, B=@n16, Alpha=@n4.G)
   ) {attributeSetTypes [#1, #2]}
  ) {materialFunction=&BaseFN_CharacterFunction_PostFX}
  Saturate(None=@n6)
 ]
}
